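# M_Food is a minimal opaque Default Lit material under gameplay effects, flagged for use with mesh particles, likely for food-themed particle meshes.
# It feeds a single constant warm peach color into the diffuse input of the ConvertFromDiffSpec material function, with no textures or parameters.
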Material /Figment_S10/Effects/Fort_Effects/Gameplay/M_Food
graph {
 properties {bUsedWithMeshParticles=true, pixelDepthOffsetMode=PDOM_Legacy, usedShadingModels=Default Lit}
 nodes [MaterialFunctionCall(DiffuseColor=Constant3Vector() {constant=<rgba 1 0.7 0.5 0>}) {materialFunction=&ConvertFromDiffSpec}]
}
